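# PhysicsAsset for the TrafficHat pickaxe cosmetic, a two-handed melee weapon mesh.
# It defines a single capsule body on the Melee_TwoHanded_Handle bone, offset and rotated along the axe.
# It has no constraints, so the weapon acts as one rigid collision shape.
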PhysicsAsset /BRCosmetics/Weapons/FORT_Melee/Pickaxe_TrafficHat/Meshes/TrafficHat_Axe_PhysicsAsset
bodies {Melee_TwoHanded_Handle [{shapeType=Capsule, length=73.6592, radius=21.7281, center=<0.3284 -7.7863 30.4478>, rotation=<rot 1.1268 -90 0.0563>}]}
constraints []
